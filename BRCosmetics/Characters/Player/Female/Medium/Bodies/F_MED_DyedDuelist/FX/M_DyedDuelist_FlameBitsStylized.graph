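# Translucent, unlit, two-sided Niagara sprite material that draws stylized flame bits for the Dyed Duelist character's effects.
# A slowly panning cloud noise, clamped to keep a brightness floor, modulates the red channel of a flame texture, while its green channel blends the result toward white.
# That shape is multiplied by particle color for an emissive adjustment function, and separately by particle alpha.
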Material /BRCosmetics/Characters/Player/Female/Medium/Bodies/F_MED_DyedDuelist/FX/M_DyedDuelist_FlameBitsStylized
graph {
 properties {
  bUsedWithNiagaraSprites=true
  blendMode=BLEND_Translucent
  pixelDepthOffsetMode=PDOM_Legacy
  shadingModel=MSM_Unlit
  twoSided=true
  usedShadingModels=Unlit
 }
 nodes [
  n1 = TextureSample(UVs=Panner(Coordinate=TextureCoordinate() {uTiling=0.25, vTiling=0.25}) {speedX=0.25, speedY=0.1}) {texture=&AtmosphericCloudNoise01}
  n2 = TextureSample(UVs=TextureCoordinate()) {
   samplerType=SAMPLERTYPE_LinearColor
   texture=&/T_DyedDuelist_FlameStylized
  }
  n3 = LinearInterpolate(A=Multiply(A=Clamp(None=@n1.RGB) {minDefault=0.5}, B=@n2.R), Alpha=Multiply(A=@n2.G))
  n4 = ParticleColor()
  MaterialFunctionCall(Emissive=Multiply(A=@n3, B=@n4.RGB)) {materialFunction=&MF_FX_EmissiveAdjustments}
  Multiply(A=@n3, B=@n4.A)
 ]
}
