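# Niagara reactive effect for the Dyed Duelist cosmetic body, producing timed purple-magenta flames on skeletal mesh bones that fade out over a few seconds.
# Flame sprites with green cores, rising curl-noise fire bits and four animation-trail ribbons attach to bone or socket indices, alongside two fading lights.
# User.bisHandHidden hides selected sprite partitions through visibility tags and kills the matching ribbons.
# Fire-bit saturation and brightness use front-end versus gameplay modifiers.
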NiagaraSystem /BRCosmetics/Characters/Player/Female/Medium/Bodies/F_MED_DyedDuelist/FX/NS_DyedDuelist_Reactive
dependencies {
 usedDataInterfaces [
  &NiagaraDataInterfaceArrayFloat
  &NiagaraDataInterfaceCurve
  &NiagaraDataInterfaceArrayInt32
  &NiagaraDataInterfaceArrayFloat3
  &NiagaraDataInterfaceSpriteRendererInfo
 ]
 usedDynamicInputs [&Multiply_Vector, &NormalizeVector]
 usedModules [
  &SystemState
  &self:SystemSpawnScript.SystemScriptSource.SystemScriptGraph.NiagaraNodeAssignment_0.SetVariables_#1
  &InterpolateOverTime
  &self:SystemSpawnScript.SystemScriptSource.SystemScriptGraph.NiagaraNodeAssignment_1.SetVariables_#2
 ]
}
emitters {
 FireBits {
  emitterUpdateScript [
   &EmitterState(Life Cycle Mode=::NewEnumerator1, Loop Behavior=::NewEnumerator1, Loop Duration=2.7, UseLoopDelay=false)
   &SpawnBurst_Instantaneous(
    Spawn Count=&RandomRangeInt(Minimum=4, Maximum=8)
    Use Spawn Probability=false
    UseLoopCountLimit=false
   )
   &SpawnRate(SpawnRate=&RandomRangeFloat(Minimum=3, Maximum=6), Use Spawn Probability=false)
  ]
  particleSpawnScript [
   &InitializeParticle(
    AdjustAlpha=false
    AdjustHue=true
    AdjustSaturation=false
    AdjustValue=false
    Color=<rgba 0.1756 0 0.3219 1>
    Color Mode=::NewEnumerator3
    Hue Shift Range:Vector2f=<-0.0462 0>
    Lifetime Max=0.8
    Lifetime Min=0.5
    Lifetime Mode=::NewEnumerator1
    Sprite Size=&Multiply_Vector2D(
     A=&RandomRangeVector2D(Minimum=<1 3>, Maximum=<2.5 5>)
     B=&Vector2DFromFloat(Value=5)
    )
    Sprite Size Mode=::NewEnumerator4
    UsePositionOffset=false
    Write Material Random=true
   )
   &PartitionParticles(Number Of Partitions=4, Partition Mode:ENiagara_PartitionMode=::NewEnumerator5)
   &SkeletalMeshLocation(
    Bone / Socket Index=&SelectIntFromArray(Direct Array Index=$Particles.PartitionParticles.Partition, Int Selection Array=DataInterface() {propertyValues {IntData [0,1,2,3]}})
    Bone Sampling Mode=::NewEnumerator4
    MaskSpawnGroup=false
    Sampled Position Offset=&Lerp_Vector(
     A=&SelectVectorFromArray(
      Direct Array Index=$Particles.PartitionParticles.Partition
      Vector Selection Array=DataInterface() {propertyValues {FloatData [<0 0 0>, <0 0 0>, <0 0 0>, <0 0 0>]}}
     )
     B=&SelectVectorFromArray(
      Direct Array Index=$Particles.PartitionParticles.Partition
      Vector Selection Array=DataInterface() {propertyValues {FloatData [<-33 0 0>, <-33 0 0>, <33 0 0>, <33 0 0>]}}
     )
     Alpha=&RandomRangeFloat(Minimum=0, Maximum=1)
    )
    Skeletal Mesh=$User.Skeletal Mesh
   )
   &AddVelocity(
    Velocity=&RandomRangeVector(Minimum=<-0.5 -0.5 1>, Maximum=<0.5 0.5 1>)
    Velocity Speed Scale=&RandomRangeFloat(Minimum=40, Maximum=60)
   )
  ]
  particleUpdateScript [
   &ParticleState()
   &ScaleSpriteSize(Uniform Curve Sprite Scale=DataInterface() {propertyValues {Curve {keys [{interpMode=RCIM_Cubic, tangentMode=RCTM_User, time=0.3, value=1}, {tangentMode=RCTM_User, time=1, value=0}]}}}) {module=ScaleSpriteSize001}
   &ScaleSpriteSize(Uniform Curve Sprite Scale=DataInterface() {propertyValues {Curve {keys [
        {arriveTangent=3.2867, interpMode=RCIM_Cubic, leaveTangent=3.2867, tangentMode=RCTM_User, time=0, value=0}
        {interpMode=RCIM_Cubic, tangentMode=RCTM_User, time=1, value=1}
       ]}}})
   &GravityForce(Gravity=<0 0 50>)
   &Drag(Drag=2, Use Linear Drag=true, Use Rotational Drag=true)
   &CurlNoiseForce(Noise Frequency=100, Noise Strength=50, Pan Noise=true, Pan Noise Field=<1 0 0>)
   &SolveForcesAndVelocity(Clamp Velocity=false, Limit Acceleration=false)
   &SubUVAnimation(
    Evaluation Type=::NewEnumerator0
    "Sprite Renderer "=DataInterface() {propertyValues {SpriteRenderer=&self:FireBits_4.NiagaraSpriteRendererProperties_1}}
    SubUV Animation Mode:ENiagara_SubUVLookupModeV2=::NewEnumerator2
    UseEndFrame=false
    UseStartFrame=false
   )
   &Color(Color=&HueShiftLinearColor(
     Hue=0
     Saturation=&FN_FrontEndFloatModifier(
      Gameplay=&FloatFromCurve(
       FloatCurve=DataInterface() {propertyValues {Curve {keys [
           {arriveTangent=-1.954, interpMode=RCIM_Cubic, leaveTangent=-1.954, tangentMode=RCTM_User, time=0, value=1.7}
           {arriveTangent=-0.0364, interpMode=RCIM_Cubic, tangentMode=RCTM_Break, time=0.4, value=1}
           {interpMode=RCIM_Cubic, tangentMode=RCTM_User, time=1, value=1}
          ]}}}
       CurveIndex=$Emitter.Age
      )
      Front End=0
     )
     Value=&FN_FrontEndFloatModifier(
      Gameplay=&FloatFromCurve(FloatCurve=DataInterface() {propertyValues {Curve {keys [
           {arriveTangent=0.3544, interpMode=RCIM_Cubic, leaveTangent=0.3544, tangentMode=RCTM_User, time=0, value=2}
           {arriveTangent=0.4308, interpMode=RCIM_Cubic, tangentMode=RCTM_Break, time=0.1, value=1}
           {interpMode=RCIM_Cubic, tangentMode=RCTM_User, time=1, value=1}
          ]}}})
      Front End=1
     )
     Linear Color=$Particles.Color
    )) {module=Color001}
   /SetVariables(Particles.VisibilityTag=&MakeCustomIntFromBool(
     Bool=&Boolean_AndOperation(
      Bool1=$User.bisHandHidden
      Bool2=&Boolean_OrOperation(
       Bool1=&SetBoolByIntComparison(A=$Particles.PartitionParticles.Partition, B=1, Comparison Type=::NewEnumerator4)
       Bool2=&SetBoolByIntComparison(A=$Particles.PartitionParticles.Partition, B=3, Comparison Type=::NewEnumerator4)
      )
     )
     True INT=2
     False INT=0
    )) {isSetParameters=true, module=SetVariables_#3}
  ]
  rendererClasses [&NiagaraSpriteRendererProperties]
  renderers [{rendererClass=&NiagaraSpriteRendererProperties}]
 }
 Flames_Ribbons {
  emitterUpdateScript [
   &EmitterState(Life Cycle Mode=::NewEnumerator1, Loop Behavior=::NewEnumerator1, Loop Duration=3.7, UseLoopDelay=false)
   &FN_SpawnAnimTrail(Number of Anim Trails=4, Spawn Mode:ENiagaraSkeletalSpawning=::NewEnumerator0)
  ]
  particleSpawnScript [
   &FN_InitializeAnimTrailBoneIndecies()
   &FN_InitializeAnimTrailBoneIndecies(AnimTrail Selection=1, First Bone or Socket Index=1, Second Bone or Socket Index=2) {module=FN_InitializeAnimTrailBoneIndecies001}
   &FN_InitializeAnimTrailBoneIndecies(AnimTrail Selection=2, First Bone or Socket Index=3, Second Bone or Socket Index=4) {module=FN_InitializeAnimTrailBoneIndecies002}
   &FN_InitializeAnimTrailBoneIndecies(AnimTrail Selection=3, First Bone or Socket Index=4, Second Bone or Socket Index=5) {module=FN_InitializeAnimTrailBoneIndecies003}
   &FN_InitializeAnimTrail(
    AnimTrailWidth=&SelectFloatFromArray(Direct Array Index=$Particles.SpawnGroup, Float Selection Array=DataInterface() {propertyValues {FloatData [1.4,1,1.4,1]}})
    Lifetime=0.3
    Position Offset=&Multiply_VectorByFloat(
     Vector=<2 0 0>
     Float=&SelectFloatFromArray(Direct Array Index=$Particles.SpawnGroup, Float Selection Array=DataInterface() {propertyValues {FloatData [-1,-1,1,1.7]}})
    )
    SkeletalMesh=$User.SkeletalMeshV2
   )
   &AddVelocity(
    Velocity=<100 0 0>
    Velocity Speed Scale=&SelectFloatFromArray(Direct Array Index=$Particles.SpawnGroup, Float Selection Array=DataInterface() {propertyValues {FloatData [-1,-1,1,1]}})
   )
   &AddVelocity(Velocity=<0 0 222>) {module=AddVelocity002}
  ]
  particleUpdateScript [
   &ParticleState()
   &ScaleColor(
    Scale Alpha=&FloatFromCurve(FloatCurve=DataInterface() {propertyValues {Curve {keys [<at -0.0046 1.0113 RCIM_Cubic>,<at 1.0011 0.002 RCIM_Cubic>]}}})
    ScaleA=true
    ScaleRGB=false
   )
   &FN_UpdateAnimTrail(SkeletalMesh=$User.Skeletal Mesh)
   &SolveForcesAndVelocity(Clamp Velocity=false, Limit Acceleration=false)
   &CameraOffset(Camera Offset Amount=3)
   &ScaleColor(
    Scale Alpha=&FloatFromCurve(
     FloatCurve=DataInterface() {propertyValues {Curve {keys [
         {tangentMode=RCTM_User, time=0, value=1}
         {arriveTangent=-0.0405, interpMode=RCIM_Cubic, leaveTangent=-0.0405, tangentMode=RCTM_User, time=0.7009, value=1}
         <at 2.1 1>
         {arriveTangent=-1.8979, interpMode=RCIM_Cubic, leaveTangent=-1.8979, tangentMode=RCTM_User, time=3.3, value=0}
        ]}}}
     CurveIndex=$Emitter.Age
    )
    ScaleA=true
    ScaleRGB=false
   ) {module=ScaleColor001}
   &KillParticles(Kill Particles=&Boolean_AndOperation(
     Bool1=$User.bisHandHidden
     Bool2=&Boolean_OrOperation(
      Bool1=&SetBoolByIntComparison(A=$Particles.FirstAnimTrailBone, B=1, Comparison Type=::NewEnumerator4)
      Bool2=&SetBoolByIntComparison(A=$Particles.FirstAnimTrailBone, B=4, Comparison Type=::NewEnumerator4)
     )
    ))
  ]
  rendererClasses [&NiagaraRibbonRendererProperties]
  renderers [{rendererClass=&NiagaraRibbonRendererProperties}]
 }
 Flames_Sprites {
  emitterUpdateScript [
   &EmitterState(Inactive Response=::NewEnumerator1, Life Cycle Mode=::NewEnumerator1, Loop Behavior=::NewEnumerator1, Loop Duration=2.7, UseLoopDelay=false)
   &SpawnRate(SpawnRate=&RandomRangeFloat(Minimum=1, Maximum=3), Use Spawn Probability=false)
   &SpawnBurst_Instantaneous(
    Spawn Count=&RandomRangeInt(Minimum=4, Maximum=6)
    Use Spawn Probability=false
    UseLoopCountLimit=false
   )
  ]
  particleSpawnScript [
   &InitializeParticle(
    Color=<rgba 0.5042 0 1 1>
    Lifetime Max=3
    Lifetime Mode=::NewEnumerator1
    Sprite Size=&Multiply_Vector2D(
     A=&RandomRangeVector2D(Minimum=<7 7>, Maximum=<10 9>)
     B=&Vector2DFromFloat(Value=4)
    )
    Sprite Size Mode=::NewEnumerator4
    UsePositionOffset=false
    Write Material Random=true
   )
   &PartitionParticles(Number Of Partitions=4, Partition Mode:ENiagara_PartitionMode=::NewEnumerator5)
  ]
  particleUpdateScript [
   &ParticleState()
   &SkeletalMeshLocation(
    Bone / Socket Index=$Particles.PartitionParticles.Partition
    Bone Sampling Mode=::NewEnumerator4
    MaskSpawnGroup=false
    Sampled Position Offset=&Lerp_Vector(
     A=&SelectVectorFromArray(
      Direct Array Index=$Particles.PartitionParticles.Partition
      Vector Selection Array=DataInterface() {propertyValues {FloatData [<0 0 0>, <0 0 0>, <0 0 0>, <0 0 0>]}}
     )
     B=&SelectVectorFromArray(
      Direct Array Index=$Particles.PartitionParticles.Partition
      Vector Selection Array=DataInterface() {propertyValues {FloatData [<-22 0 0>, <-22 0 0>, <22 0 0>, <22 0 0>]}}
     )
     Alpha=&RandomRangeFloat(Minimum=0, Maximum=1)
    )
    Skeletal Mesh=$User.Skeletal Mesh
   ) {module=SkeletalMeshLocation001}
   &SpriteFacingAndAlignment(
    Sprite Alignment=&Add_Vector(
     A=$System.DiveVelocityScale
     B=&Multiply_VectorByFloat(
      Vector=&ReturnMeshOrientationAxis(Mesh Orientation Axis=::ZAxis, Mesh Orientation Quat=$Particles.MeshOrientation)
      Float=&SelectFloatFromArray(
       Direct Array Index=$Particles.PartitionParticles.Partition
       Float Selection Array=DataInterface() {propertyValues {FloatData [-0.5,-0.5,0.5,0.5]}}
      )
     )
    )
    WriteAlignment=true
    WriteFacing=false
   )
   &SolveForcesAndVelocity(Clamp Velocity=false, Limit Acceleration=false)
   /SetVariables(Particles.FlameCoreSize:Vector2f=&Multiply_Vector2D(A=<8 8>, B=&Vector2DFromFloat(Value=2))) {isSetParameters=true}
   /SetVariables(Particles.FlameCoreColor:LinearColor=&MakeLinearColor(
     R:NiagaraFloat=0.2852
     G:NiagaraFloat=1
     B:NiagaraFloat=0.6816
     A:NiagaraFloat=&Multiply_Float(
      A=&FloatFromCurve(
       FloatCurve=DataInterface() {propertyValues {Curve {keys [
           {interpMode=RCIM_Cubic, tangentMode=RCTM_User, time=0, value=1}
           {arriveTangent=-1.5786, interpMode=RCIM_Cubic, leaveTangent=-1.5786, tangentMode=RCTM_User, time=2.5, value=0}
          ]}}}
       CurveIndex=$Emitter.Age
      )
      B=&FloatFromCurve(FloatCurve=DataInterface() {propertyValues {Curve {keys [
           {interpMode=RCIM_Cubic, tangentMode=RCTM_User, time=0, value=1}
           {arriveTangent=-3.5614, interpMode=RCIM_Cubic, leaveTangent=-3.5614, tangentMode=RCTM_User, time=1, value=0}
          ]}}})
     )
    )) {isSetParameters=true}
   &DynamicMaterialParameters(
    Index 0 Param 1=&RandomRangeFloat(Minimum=5.9615, Maximum=222)
    Module.Index0Param0Write=true
    Module.Index0Param1Write=false
    Module.Index0Param2Write=false
    Module.Index0Param3Write=false
   )
   &CameraOffset(Camera Offset Amount=&RandomRangeFloat(Minimum=0, Maximum=4))
   &ScaleColor(
    Scale Alpha=&FloatFromCurve(FloatCurve=DataInterface() {propertyValues {Curve {keys [
         {tangentMode=RCTM_User, time=0, value=0}
         <at 0.13 1>
         {arriveTangent=-5.6696, interpMode=RCIM_Cubic, leaveTangent=-5.6696, tangentMode=RCTM_User, time=0.8, value=1}
         {arriveTangent=-0.4419, interpMode=RCIM_Cubic, leaveTangent=-0.4419, tangentMode=RCTM_User, time=1, value=0}
        ]}}})
    ScaleA=true
    ScaleRGB=false
   )
   &ScaleColor(
    Scale Alpha=&FloatFromCurve(
     FloatCurve=DataInterface() {propertyValues {Curve {keys [
         <at 0 1>
         {interpMode=RCIM_Cubic, tangentMode=RCTM_Break, time=2, value=1}
         {interpMode=RCIM_Cubic, tangentMode=RCTM_Break, time=2.7, value=0}
        ]}}}
     CurveIndex=$Emitter.Age
    )
    ScaleA=true
    ScaleRGB=false
   ) {module=ScaleColor001}
   /SetVariables(Particles.VisibilityTag=&MakeCustomIntFromBool(
     Bool=&Boolean_AndOperation(
      Bool1=$User.bisHandHidden
      Bool2=&Boolean_OrOperation(
       Bool1=&SetBoolByIntComparison(A=$Particles.PartitionParticles.Partition, B=1, Comparison Type=::NewEnumerator4)
       Bool2=&SetBoolByIntComparison(A=$Particles.PartitionParticles.Partition, B=3, Comparison Type=::NewEnumerator4)
      )
     )
     True INT=2
     False INT=0
    )) {isSetParameters=true, module=SetVariables_#3}
  ]
  rendererClasses [&NiagaraSpriteRendererProperties]
  renderers [
   {rendererClass=&NiagaraSpriteRendererProperties}
   {rendererClass=&NiagaraSpriteRendererProperties, rendererIndex=1}
  ]
 }
 Light {
  emitterUpdateScript [
   &EmitterState(Life Cycle Mode=::NewEnumerator1, Loop Behavior=::NewEnumerator1, Loop Duration=4, UseLoopDelay=false)
   &SpawnBurst_Instantaneous(Spawn Count=2, Use Spawn Probability=false, UseLoopCountLimit=false)
  ]
  particleSpawnScript [
   &InitializeParticle(Color=<rgba 0.9203 0 2 3>, Lifetime=2.9, UsePositionOffset=false, Write Material Random=true)
   &PartitionParticles()
  ]
  particleUpdateScript [
   &ParticleState()
   &SkeletalMeshLocation(
    Bone / Socket Index=&SelectIntFromArray(Direct Array Index=$Particles.PartitionParticles.Partition, Int Selection Array=DataInterface() {propertyValues {IntData [0,2]}})
    Bone Sampling Mode=::NewEnumerator4
    MaskSpawnGroup=false
    Sampled Position Offset=&SelectVectorFromArray(
     Direct Array Index=$Particles.PartitionParticles.Partition
     Vector Selection Array=DataInterface() {propertyValues {FloatData [<-18 0 0>, <18 0 0>]}}
    )
    Skeletal Mesh=$User.Skeletal Mesh
   )
   &ScaleColor(
    Scale Alpha=&FloatFromCurve(FloatCurve=DataInterface() {propertyValues {Curve {keys [
         {interpMode=RCIM_Cubic, tangentMode=RCTM_User, time=0, value=1}
         {arriveTangent=-3.4397, interpMode=RCIM_Cubic, leaveTangent=-3.4397, tangentMode=RCTM_User, time=1, value=0}
        ]}}})
    ScaleA=true
    ScaleRGB=false
   )
   /SetVariables(
    Particles.PointLightColor:LinearColor=<rgba 0.9216 0 1 1>
    Particles.PointLightIntensity:NiagaraFloat=&FloatFromCurve(
     FloatCurve=DataInterface() {propertyValues {Curve {keys [
         {interpMode=RCIM_Cubic, tangentMode=RCTM_User, time=0, value=1}
         {arriveTangent=-3.4397, interpMode=RCIM_Cubic, leaveTangent=-3.4397, tangentMode=RCTM_User, time=1, value=0}
        ]}}}
     Scale Curve=10.219
    )
   ) {isSetParameters=true}
  ]
  rendererClasses [&NiagaraLightRendererProperties, &NiagaraComponentRendererProperties]
  renderers [
   {rendererClass=&NiagaraLightRendererProperties, rendererIndex=0}
   {rendererClass=&NiagaraComponentRendererProperties, rendererIndex=1}
  ]
 }
}
summary {
 systemName=NS_DyedDuelist_Reactive
 emitters [
  {emitterName=Flames_Ribbons, rendererClasses [&NiagaraRibbonRendererProperties]}
  {emitterName=Light, rendererClasses [&NiagaraLightRendererProperties, &NiagaraComponentRendererProperties]}
  {emitterName=FireBits, rendererClasses [&NiagaraSpriteRendererProperties]}
  {emitterName=Flames_Sprites, rendererClasses [&NiagaraSpriteRendererProperties]}
 ]
 userVariables [
  User.bIsFrontEnd {type=NiagaraBool}
  User.bIsFrontEndPreview {type=NiagaraBool}
  User.bisHandHidden {type=NiagaraBool}
  User.RibbonIDOffset {type=NiagaraInt32}
  User.Skeletal Mesh {
   type=NiagaraDataInterfaceSkeletalMesh
   defaultValue {
    struct=NiagaraExt_VariableValue_DataInterface
    value {dataInterface=&self:NiagaraDataInterfaceSkeletalMesh_0}
   }
  }
  User.SkeletalMeshV2 {
   type=NiagaraDataInterfaceSkeletalMesh
   defaultValue {
    struct=NiagaraExt_VariableValue_DataInterface
    value {dataInterface=&self:NiagaraDataInterfaceSkeletalMesh_2}
   }
  }
 ]
}
