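# Skeletal mesh body material for the DyedDuelist female character cosmetic, adding animated-style swirl, gem glow and fresnel effects to the shared character shading.
# It edits Base Color and Emissive between the MF_CharacterParent_2023 PreFX and PostFX functions.
# The swirl function borrowed from ChicleVeil blends into both channels, limited by the Emissive texture's green mask and power curves.
# A body emissive texture masks the tinted gem glow.
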Material /BRCosmetics/Characters/Player/Female/Medium/Bodies/F_MED_DyedDuelist/Materials/M_DyedDuelist_Body
graph {
 properties {
  bUseMaterialAttributes=true
  bUsedWithMeshDeformer=true
  bUsedWithMorphTargets=true
  bUsedWithNanite=true
  bUsedWithSkeletalMesh=true
  pixelDepthOffsetMode=PDOM_Legacy
  previewMesh=/Game/Characters/Player/Female/Medium/Bodies/F_MED_DyedDuelist/Meshes/F_MED_DyedDuelist
  shadingModel=MSM_FromMaterialExpression
  usedShadingModels=Default Lit | Subsurface | Clear Coat | Subsurface Profile | Cloth
 }
 nodes [
  n1 = NamedRerouteUsage()
  n2 = NamedRerouteUsage()
  n3 = TextureSampleParameter2D() {group=06_Emissive, parameterName=Emissive, sortPriority=1, texture=&T_BlacksRGB}
  n4 = SmoothStep(Value=@n3.G) {constMax=0.5}
  n5 = Max(
   A=Power(Base=OneMinus(None=@n2), Exp=ScalarParameter() {defaultValue=3, parameterName=EmissiveMittigationPower}) {constExponent=1}
   B=OneMinus(None=@n4)
  )
  n6 = Multiply(
   A=VectorParameter() {parameterName=GemsColor, defaultValue=<rgba 0.1672 0 1 1>}.RGB
   B=ScalarParameter() {defaultValue=1.5, parameterName=GemEmissiveIntensity}
  )
  n7 = TextureSample() {
   samplerSource=SSM_Wrap_WorldGroupSettings
   texture=/Game/Characters/Player/Female/Medium/Bodies/F_MED_DyedDuelist/Textures/T_DyedDuelist_Body_E
  }
  n8 = Step(X=@n7.R) {constX=0.1, constY=0.1}
  n9 = Add(
   A=Multiply(A=Multiply(A=@n1, B=@n5), B=ScalarParameter() {defaultValue=1, parameterName=BaseEmissive})
   B=Add(A=Multiply(A=@n6, B=@n8), B=NamedRerouteUsage())
  )
  Add(A=@n9, B=NamedRerouteUsage())
  Desaturation(None=@n1, Fraction=Multiply(A=ScalarParameter() {parameterName=EmissiveDesat}, B=NamedRerouteUsage()))
  n10 = MaterialFunctionCall() {materialFunction=&MF_CharacterParent_2023_PreFX}
  MaterialFunctionCall(Material=NamedRerouteUsage(), M=@n10.M, Pre Skinned Local Position=@n10.M, Distance Blend=@n10.M, Fresnel=@n10.M) {materialFunction=&MF_CharacterParent_2023_PostFX}
  n11 = GetMaterialAttributes(None=@n10.Result) {attributeGetTypes [#1, #2]}
  NamedRerouteDeclaration(None=@n11.Base Color) {name=InBaseColor, nodeColor=<rgba 1 0 0.7031 1>}
  NamedRerouteDeclaration(None=@n11.Emissive Color) {name=InEmissiveColor, nodeColor=<rgba 0 1 0.1562 1>}
  n12 = Add(A=TextureCoordinate(), B=Constant2Vector() {g=-0.17, r=-0.2})
  n13 = Constant2Vector() {g=0.3, r=0.32}
  n14 = NamedRerouteUsage()
  n15 = Multiply(
   A=Multiply(
    A=MaterialFunctionCall(UVs=@n12, Texture Scale=@n13) {materialFunction=&ScaleUVsByCenter}.0-1 mask
    B=MaterialFunctionCall(Float3=@n14) {materialFunction=&BreakOutFloat3Components}.B
   )
   B=VectorParameter() {parameterName=GlowFresnelColor, defaultValue=<rgba 0.5195 0 1 1>}.RGB
  )
  n16 = ScalarParameter() {defaultValue=1.5, parameterName=SleeveShapeOutlineIntensity}
  NamedRerouteDeclaration(None=Multiply(A=Multiply(A=@n15, B=@n16), B=ScalarParameter() {parameterName=SleeveShapeOutlineBlend})) {name=SleeveShapeOutline, nodeColor=<rgba 0.7109 0 1 1>}
  NamedRerouteDeclaration(None=@n11.MaterialAttributes) {name=InMA, nodeColor=<rgba 1 0 0.75 1>}
  n17 = NamedRerouteUsage()
  n18 = Power(Base=OneMinus(None=NamedRerouteUsage()), Exp=ScalarParameter() {defaultValue=2, group=Swirl, parameterName=SwirlFresnelPower}) {constExponent=8}
  n19 = ScalarParameter() {defaultValue=1, group=Swirl, parameterName=SwirlFresnelMult}
  n20 = MaterialFunctionCall!(
   FresnelMax=Constant()
   SwirlColorBias=ScalarParameter() {group=Swirl, parameterName=SwirlBias}
   SwirlFresnel=NamedRerouteDeclaration(None=Multiply(A=@n18, B=@n19)) {name=SwirlFresnel, nodeColor=<rgba 1 0 0.6562 1>}
   SwirlColorBalanceOverride
   SwirlColorBalanceOverrideMask
  ) {materialFunction=/BRCosmetics/Characters/Player/Female/Medium/Bodies/F_MED_ChicleVeil/Materials/MF_ChicleVeil_TA}
  n21 = Multiply(A=@n20.Swirl, B=ScalarParameter() {defaultValue=0.5, parameterName=SwirlBaseColorIntensity})
  n22 = Multiply(A=@n20.SwirlAlpha, B=@n4)
  n23 = Min(
   A=@n22
   B=Power(Base=@n2, Exp=ScalarParameter() {defaultValue=1, group=Swirl, parameterName=SwirlBaseColorBlendFresnelExponent}) {constExponent=4}
  )
  n24 = Multiply(A=@n20.Swirl, B=ScalarParameter() {defaultValue=1, parameterName=SwirlEmissiveIntensity})
  n25 = Min(
   A=@n22
   B=Power(Base=@n2, Exp=ScalarParameter() {defaultValue=3, group=Swirl, parameterName=SwirlEmissiveBlendFresnelExponent}) {constExponent=4}
  )
  NamedRerouteDeclaration(None=SetMaterialAttributes!(
    MaterialAttributes=NamedRerouteUsage()
    Base Color=LinearInterpolate(A=@n17, B=@n21, Alpha=@n23)
    Emissive Color=LinearInterpolate(A=@n9, B=@n24, Alpha=@n25)
   ) {attributeSetTypes [#1, #2]}) {name=OutMA, nodeColor=<rgba 0 0.1562 1 1>}
  NamedRerouteDeclaration(None=@n10.Tech Art Mask) {name=InTAMask, nodeColor=<rgba 0 1 0.6953 1>}
  NamedRerouteDeclaration(None=@n10.Fresnel) {name=InFresnel, nodeColor=<rgba 1 0.3047 0 1>}
  NamedRerouteDeclaration(None=@n10.Pre skinned Local Position) {name=PreSkinPos, nodeColor=<rgba 1 0 0.2109 1>}
  n26 = NamedRerouteUsage()
  NamedRerouteDeclaration(None=Multiply(A=ComponentMask(None=@n26) {r=true}, B=@n20.FesnelGlow)) {name=FresnelGlow, nodeColor=<rgba 0.0078 1 0 1>}
  n27 = Power(Base=Multiply(A=Ceil(None=@n3.R), B=OneMinus(None=@n3.R))) {constExponent=32}
  NamedRerouteDeclaration(None=Saturate(None=Multiply(A=@n27) {constB=24})) {name=PurpleGradients, nodeColor=<rgba 1 0.8203 0 1>}
  TextureSample() {texture=/BRCosmetics/Characters/Player/Female/Medium/Bodies/F_MED_DyedDuelist/Skins/Stage2/Textures/T_DyedDuelist_Stage2_Body_E}
  TextureSample() {
   samplerSource=SSM_Wrap_WorldGroupSettings
   samplerType=SAMPLERTYPE_Masks
   texture=/BRCosmetics/Characters/Player/Female/Medium/Bodies/F_MED_DyedDuelist/Skins/Stage2/Textures/T_DyedDuelist_Stage2_Body_FX
  }
 ]
}
