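# Physics asset for the F_MED_DyedDuelist face accessory mesh that sets up left and right ponytails with ribbons for physical simulation.
# Each six-capsule ponytail chains from a zero-size head anchor with limits loosest at root and tip, ending in a three-body ribbon with twist locked and hinge-like lower joints.
# Torso, neck, clavicle and upper-arm capsules, linked by free-rotation constraints, appear to serve as colliders, with oversized shapes on neck_01 and spine_05.
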PhysicsAsset /BRCosmetics/Characters/Player/Female/Medium/Bodies/F_MED_DyedDuelist/Meshes/Parts/F_MED_DyedDuelist_FaceAcc_Physics
bodies {
 clavicle_l [{shapeType=Capsule, length=15, radius=8.5, shapeName=clavicle_r_capsule, center=<7.5745 1.5 -3.491>, rotation=<rot 82.3932 0 0>}]
 clavicle_r [{shapeType=Capsule, length=15, radius=8.5, shapeName=clavicle_r_capsule, center=<-7.5745 -1.5 3.491>, rotation=<rot 82.3932 0 0>}]
 dyn_L_ponytailRibbon_A [{shapeType=Capsule, length=3, radius=2, shapeName=dyn_L_ponytailRibbon_B_01_capsule, center=<2.1585 -1.3 0>, rotation=<rot 90 0 0>}]
 dyn_L_ponytailRibbon_B_01 [{shapeType=Capsule, length=3, radius=2, shapeName=dyn_L_ponytailRibbon_B_01_capsule, center=<2.1585 -1.3 0>, rotation=<rot 90 0 0>}]
 dyn_L_ponytailRibbon_B_02 [{shapeType=Capsule, length=3, radius=2, shapeName=dyn_L_ponytailRibbon_B_02_capsule, center=<2.1585 -1.3 0>, rotation=<rot 90 0 0>}]
 dyn_L_ponytail_01 [{shapeType=Capsule, length=7, radius=3, shapeName=dyn_L_ponytail_01_capsule, center=<2 0 0>, rotation=<rot 90 0 0>}]
 dyn_L_ponytail_02 [{shapeType=Capsule, length=7, radius=1.9, shapeName=dyn_L_ponytail_02_capsule, center=<2 0 0>, rotation=<rot 87.2518 108.7086 115.9971>}]
 dyn_L_ponytail_03 [{shapeType=Capsule, length=7, radius=1.7, shapeName=dyn_L_ponytail_03_capsule, center=<2 0 0>, rotation=<rot 90 0 0>}]
 dyn_L_ponytail_04 [{shapeType=Capsule, length=7, radius=1.6, shapeName=dyn_L_ponytail_04_capsule, center=<2 0 0>, rotation=<rot 90 0 0>}]
 dyn_L_ponytail_05 [{shapeType=Capsule, length=7, radius=1.1886, shapeName=dyn_L_ponytail_05_capsule, center=<2 0 0>, rotation=<rot 90 0 0>}]
 dyn_L_ponytail_06 [{shapeType=Capsule, length=5, radius=1, shapeName=dyn_L_ponytail_06_capsule, center=<2 0 0>, rotation=<rot 90 0 0>}]
 dyn_R_ponytailRibbon_A [{shapeType=Capsule, length=3, radius=2, shapeName=dyn_R_ponytailRibbon_B_01_capsule, center=<-2.1585 1.3 0>, rotation=<rot 90 0 0>}]
 dyn_R_ponytailRibbon_B_01 [{shapeType=Capsule, length=3, radius=2, shapeName=dyn_R_ponytailRibbon_B_01_capsule, center=<-2.1585 1.3 0>, rotation=<rot 90 0 0>}]
 dyn_R_ponytailRibbon_B_02 [{shapeType=Capsule, length=3, radius=2, shapeName=dyn_R_ponytailRibbon_B_02_capsule, center=<-2.1585 1.3 0>, rotation=<rot 90 0 0>}]
 dyn_R_ponytail_01 [{shapeType=Capsule, length=7, radius=3, shapeName=dyn_R_ponytail_01_capsule, center=<-2 0 0>, rotation=<rot 90 0 0>}]
 dyn_R_ponytail_02 [{shapeType=Capsule, length=7, radius=1.9, shapeName=dyn_R_ponytail_02_capsule, center=<-2 0 0>, rotation=<rot 87.2518 108.7086 115.9971>}]
 dyn_R_ponytail_03 [{shapeType=Capsule, length=7, radius=1.7, shapeName=dyn_R_ponytail_03_capsule, center=<-2 0 0>, rotation=<rot 90 0 0>}]
 dyn_R_ponytail_04 [{shapeType=Capsule, length=7, radius=1.6, shapeName=dyn_R_ponytail_04_capsule, center=<-2 0 0>, rotation=<rot 90 0 0>}]
 dyn_R_ponytail_05 [{shapeType=Capsule, length=7, radius=1.1886, shapeName=dyn_R_ponytail_05_capsule, center=<-2 0 0>, rotation=<rot 90 0 0>}]
 dyn_R_ponytail_06 [{shapeType=Capsule, length=5, radius=1, shapeName=dyn_R_ponytail_06_capsule, center=<-2 0 0>, rotation=<rot 90 0 0>}]
 head [{shapeType=Capsule, length=0, radius=0, shapeName=head_capsule}]
 neck_01 [{shapeType=Capsule, length=200, radius=30, shapeName=neck_01_capsule, center=<11.2277 25.8342 0>}]
 pelvis [{shapeType=Capsule, length=10, radius=10.5, shapeName=pelvis_capsule, center=<-3.47 -5.0765 0>}]
 spine_01 [{shapeType=Capsule, length=3.3677, radius=20, shapeName=spine_01_capsule, center=<-0.3778 9.0409 0>}]
 spine_03 [{shapeType=Capsule, length=19.731, radius=20, shapeName=spine_03_capsule, center=<-1.2928 6.6496 0>}]
 spine_05 [
  {shapeType=Capsule, length=200, radius=30, shapeName=spine_05_capsule, center=<-2.3435 20.361 0>}
  {shapeType=Capsule, length=80, radius=30, shapeName=spine_05_capsule, center=<25 0 -42>, rotation=<rot 0 0 90>}
  {shapeType=Capsule, length=80, radius=30, shapeName=spine_05_capsule, center=<25 0 42>, rotation=<rot 0 0 90>}
 ]
 upperarm_l [{shapeType=Capsule, length=20, radius=5.5, shapeName=upperarm_l_capsule, center=<12 0 0>, rotation=<rot 90 0 0>}]
 upperarm_r [{shapeType=Capsule, length=20, radius=5.5, shapeName=upperarm_l_capsule, center=<-12 0 0>, rotation=<rot 90 0 0>}]
}
constraints [
 {bone1Name=dyn_L_ponytail_02, bone2Name=dyn_L_ponytail_01, swing1LimitDegrees=20, swing2LimitDegrees=20, twistLimitDegrees=10}
 {bone1Name=dyn_L_ponytail_03, bone2Name=dyn_L_ponytail_02, swing1LimitDegrees=20, swing2LimitDegrees=20, twistLimitDegrees=10}
 {bone1Name=dyn_L_ponytail_04, bone2Name=dyn_L_ponytail_03, swing1LimitDegrees=20, swing2LimitDegrees=20, twistLimitDegrees=10}
 {bone1Name=dyn_L_ponytail_05, bone2Name=dyn_L_ponytail_04, swing1LimitDegrees=30, swing2LimitDegrees=30, twistLimitDegrees=10}
 {bone1Name=dyn_L_ponytail_06, bone2Name=dyn_L_ponytail_05, swing1LimitDegrees=30, swing2LimitDegrees=30, twistLimitDegrees=10}
 {
  bone1Name=dyn_L_ponytailRibbon_B_01
  bone2Name=dyn_L_ponytailRibbon_A
  swing1LimitDegrees=45
  swing2LimitDegrees=45
  swing2Motion=Locked
  twistLimitDegrees=45
  twistMotion=Locked
 }
 {
  bone1Name=dyn_L_ponytailRibbon_B_02
  bone2Name=dyn_L_ponytailRibbon_B_01
  swing1LimitDegrees=45
  swing2LimitDegrees=45
  swing2Motion=Locked
  twistLimitDegrees=45
  twistMotion=Locked
 }
 {bone1Name=dyn_R_ponytail_02, bone2Name=dyn_R_ponytail_01, swing1LimitDegrees=20, swing2LimitDegrees=20, twistLimitDegrees=10}
 {bone1Name=dyn_R_ponytail_03, bone2Name=dyn_R_ponytail_02, swing1LimitDegrees=20, swing2LimitDegrees=20, twistLimitDegrees=10}
 {bone1Name=dyn_R_ponytail_04, bone2Name=dyn_R_ponytail_03, swing1LimitDegrees=20, swing2LimitDegrees=20, twistLimitDegrees=10}
 {bone1Name=dyn_R_ponytail_05, bone2Name=dyn_R_ponytail_04, swing1LimitDegrees=30, swing2LimitDegrees=30, twistLimitDegrees=10}
 {bone1Name=dyn_R_ponytail_06, bone2Name=dyn_R_ponytail_05, swing1LimitDegrees=30, swing2LimitDegrees=30, twistLimitDegrees=10}
 {
  bone1Name=dyn_R_ponytailRibbon_B_01
  bone2Name=dyn_R_ponytailRibbon_A
  swing1LimitDegrees=45
  swing2LimitDegrees=45
  swing2Motion=Locked
  twistLimitDegrees=45
  twistMotion=Locked
 }
 {
  bone1Name=dyn_R_ponytailRibbon_B_02
  bone2Name=dyn_R_ponytailRibbon_B_01
  swing1LimitDegrees=45
  swing2LimitDegrees=45
  swing2Motion=Locked
  twistLimitDegrees=45
  twistMotion=Locked
 }
 {bone1Name=dyn_R_ponytail_01, bone2Name=head, swing1LimitDegrees=30, swing2LimitDegrees=30, twistLimitDegrees=10}
 {bone1Name=dyn_L_ponytail_01, bone2Name=head, swing1LimitDegrees=30, swing2LimitDegrees=30, twistLimitDegrees=10}
 {
  bone1Name=spine_01
  bone2Name=pelvis
  swing1LimitDegrees=45
  swing1Motion=Free
  swing2LimitDegrees=45
  swing2Motion=Free
  twistLimitDegrees=45
  twistMotion=Free
 }
 {
  bone1Name=spine_03
  bone2Name=spine_01
  swing1LimitDegrees=45
  swing1Motion=Free
  swing2LimitDegrees=45
  swing2Motion=Free
  twistLimitDegrees=45
  twistMotion=Free
 }
 {
  bone1Name=spine_05
  bone2Name=spine_03
  swing1LimitDegrees=45
  swing1Motion=Free
  swing2LimitDegrees=45
  swing2Motion=Free
  twistLimitDegrees=45
  twistMotion=Free
 }
 {
  bone1Name=neck_01
  bone2Name=spine_05
  swing1LimitDegrees=45
  swing1Motion=Free
  swing2LimitDegrees=45
  swing2Motion=Free
  twistLimitDegrees=45
  twistMotion=Free
 }
 {
  bone1Name=clavicle_l
  bone2Name=spine_05
  swing1LimitDegrees=45
  swing1Motion=Free
  swing2LimitDegrees=45
  swing2Motion=Free
  twistLimitDegrees=45
  twistMotion=Free
 }
 {
  bone1Name=clavicle_r
  bone2Name=spine_05
  swing1LimitDegrees=45
  swing1Motion=Free
  swing2LimitDegrees=45
  swing2Motion=Free
  twistLimitDegrees=45
  twistMotion=Free
 }
 {
  bone1Name=upperarm_l
  bone2Name=clavicle_l
  swing1LimitDegrees=45
  swing1Motion=Free
  swing2LimitDegrees=45
  swing2Motion=Free
  twistLimitDegrees=45
  twistMotion=Free
 }
 {
  bone1Name=upperarm_r
  bone2Name=clavicle_r
  swing1LimitDegrees=45
  swing1Motion=Free
  swing2LimitDegrees=45
  swing2Motion=Free
  twistLimitDegrees=45
  twistMotion=Free
 }
 {
  bone1Name=head
  bone2Name=neck_01
  swing1LimitDegrees=45
  swing1Motion=Free
  swing2LimitDegrees=45
  swing2Motion=Free
  twistLimitDegrees=45
  twistMotion=Free
 }
 {
  bone1Name=dyn_L_ponytailRibbon_A
  bone2Name=dyn_L_ponytail_06
  swing1LimitDegrees=20
  swing2LimitDegrees=20
  twistLimitDegrees=45
  twistMotion=Locked
 }
 {
  bone1Name=dyn_R_ponytailRibbon_A
  bone2Name=dyn_R_ponytail_06
  swing1LimitDegrees=20
  swing2LimitDegrees=20
  twistLimitDegrees=45
  twistMotion=Locked
 }
]
